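# Translucent, refractive glass material for the bulbs of the Creative Spooky Alphabet meshes.
# Glow fakes shine from a view-space sphere reflection map, a sun/moon highlight and directional Fresnel fill, while Fresnel blends drive color and opacity.
# Static switches add cracks (triplanar or UV normals, optional red vertex-color mask) that alter normal, color, glow and opacity.
# Optional dirt, projected on three world axes by vertex normal, tints color and raises opacity.
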Material /Game/Creative/Environments/Meshes/Spooky/Alphabet/Materials/M_CP_Spooky_Bulb_Glass_01
graph {
 properties {
  bFullyRough=true
  blendMode=BLEND_Translucent
  pixelDepthOffsetMode=PDOM_Legacy
  refractionMethod=RM_IndexOfRefraction
  usedShadingModels=Default Lit
 }
 nodes [
  n1 = Constant3Vector() {constant=<rgba 1 1 1 0>}
  n2 = Transform(None=ReflectionVectorWS()) {transformSourceType=TRANSFORMSOURCE_World, transformType=TRANSFORM_View}
  n3 = TextureObjectParameter() {
   group=Cracks
   parameterName=Cracks Normal Map
   samplerType=SAMPLERTYPE_Normal
   sortPriority=10
   texture=/Game/Weapons/WeaponSkins/Wraps/Textures/IceBreaker/T_IceCracks_Normal
  }
  n4 = TextureCoordinate()
  n5 = ScalarParameter() {defaultValue=1, group=Cracks, parameterName=Cracks Tiling, sortPriority=11}
  n6 = StaticSwitchParameter(
   True=MaterialFunctionCall(TextureObject=@n3, TextureSize=ScalarParameter() {defaultValue=256, group=Cracks, parameterName=Cracks Scale, sortPriority=11}) {materialFunction=&WorldAlignedTexture}.XYZ Texture
   False=TextureSample(UVs=Multiply(A=@n4, B=@n5), Tex=@n3).RGB
  ) {defaultValue=true, group=Cracks, parameterName=Use Triplanar Normals, sortPriority=0}
  n7 = Power(
   Base=DotProduct(A=OneMinus(None=@n2), B=@n6)
   Exp=ScalarParameter() {defaultValue=0.25, group=Cracks, parameterName=Cracks Color Exponent, sortPriority=40}
  )
  n8 = Clamp(None=TwoSidedSign())
  n9 = Clamp(None=LinearInterpolate(
    A=@n1
    B=LinearInterpolate(A=@n1, B=@n7, Alpha=@n8)
    Alpha=ScalarParameter() {defaultValue=1, group=Cracks, parameterName=Cracks Intensity, sortPriority=12}
   ))
  n10 = VertexColor()
  n11 = ScalarParameter() {defaultValue=2, desc=Controls Falloff of Vertex Mask, group=Cracks, parameterName=Cracks Vertex Mask Exponent, sortPriority=52}
  n12 = StaticSwitchParameter(True=LinearInterpolate(A=@n1, B=@n9, Alpha=Power(Base=@n10.R, Exp=@n11)), False=@n9) {desc="White = Cracks. Black = No Cracks", group=Cracks, parameterName=Vertex Masking, sortPriority=50}
  n13 = ComponentMask(None=Transform(None=ReflectionVectorWS()) {transformSourceType=TRANSFORMSOURCE_World, transformType=TRANSFORM_View}) {g=true, r=true}
  n14 = TextureSampleParameter2D(UVs=Multiply(A=ConstantBiasScale(None=@n13), B=Constant2Vector() {g=-1, r=1})) {
   group=Reflecttion Refraction
   parameterName=SphereReflectionMap
   sortPriority=60
   texture=/Game/Packages/Fortress_SharedMaterials/Textures/ReflectionMap2
  }
  n15 = Multiply(
   A=Multiply(A=@n14.RGB, B=VectorParameter() {group=Reflecttion Refraction, parameterName=Reflection Tint, defaultValue=<rgba 1 1 1 1>}.RGB)
   B=ScalarParameter() {defaultValue=0.75, group=Reflecttion Refraction, parameterName=Reflection Multiplier}
  )
  n16 = CollectionParameter() {
   collection=&FortniteMaterialParameters
   parameterName=SunAndMoonModelDirectionalVector
  }
  n17 = DotProduct(A=ComponentMask(None=@n16) {b=true, g=true, r=true}, B=ReflectionVectorWS())
  n18 = Multiply(
   A=Power(Base=Clamp(None=@n17), Exp=ScalarParameter() {defaultValue=50, group=Reflecttion Refraction, parameterName=Spec Highlight Power}) {constExponent=20}
   B=ScalarParameter() {defaultValue=1, group=Reflecttion Refraction, parameterName=Spec Intensity}
  )
  n19 = ScalarParameter() {defaultValue=6, group=Directional Fresnel Fill Light, parameterName=Fill Light Falloff Exponent}
  n20 = PixelNormalWS()
  n21 = CollectionParameter() {
   collection=&FortniteMaterialParameters
   parameterName=SunAndMoonModelDirectionalVector
  }
  n22 = Normalize(VectorInput=ComponentMask(None=Multiply(A=@n21) {constB=-1, desc=Invert Vector}) {b=true, g=true, r=true})
  n23 = PixelNormalWS()
  n24 = Multiply(
   A=Multiply(A=Fresnel(ExponentIn=@n19, Normal=@n20) {baseReflectFraction=0, desc=Fresnel Mask, exponent=6}, B=DotProduct(A=@n22, B=@n23))
   B=VectorParameter() {group=Directional Fresnel Fill Light, parameterName=Fill Light Color, defaultValue=<rgba 1 1 1 1>}.RGB
  )
  n25 = ScalarParameter() {defaultValue=2, group=Directional Fresnel Fill Light, parameterName=Fill Light Intensity}
  n26 = ReflectionVectorWS()
  n27 = Power(
   Base=ComponentMask(None=Transform(None=@n26) {transformSourceType=TRANSFORMSOURCE_World, transformType=TRANSFORM_View}) {b=true}
   Exp=ScalarParameter() {defaultValue=1, group=Color, parameterName=Darkening Exponent, sortPriority=60}
  )
  n28 = Multiply(A=Add(A=Add(A=@n15, B=@n18), B=Multiply(A=@n24, B=@n25)), B=OneMinus(None=@n27))
  Clamp(None=StaticSwitchParameter(True=Multiply(A=@n12, B=@n28), False=@n28) {group=Cracks, parameterName=Is Cracked, sortPriority=0})
  ScalarParameter() {defaultValue=0.95, group=Reflecttion Refraction, parameterName=Refraction Index, sliderMax=2}
  n29 = Constant3Vector() {constant=<rgba 0.2159 0.2159 1 0>}
  n30 = TwoSidedSign()
  n31 = LinearInterpolate(
   A=@n29
   B=LinearInterpolate(A=@n29, B=@n6, Alpha=Clamp(None=@n30))
   Alpha=ScalarParameter() {defaultValue=1, group=Cracks, parameterName=Cracks Intensity, sortPriority=12}
  )
  n32 = ScalarParameter() {defaultValue=2, desc=Controls Falloff of Vertex Mask, group=Cracks, parameterName=Cracks Vertex Mask Exponent, sortPriority=52}
  n33 = Power(Base=VertexColor().R, Exp=@n32)
  StaticSwitchParameter(
   True=StaticSwitchParameter(True=LinearInterpolate(A=@n29, B=@n31, Alpha=@n33), False=@n31) {desc="White = Cracks. Black = No Cracks", group=Cracks, parameterName=Vertex Masking, sortPriority=50}
   False=@n29
  ) {group=Cracks, parameterName=Is Cracked, sortPriority=0}
  n34 = ScalarParameter() {defaultValue=2, desc=Determines blend between the Inner and Outer Colors, group=Color, parameterName=Color Fresnel Exponent}
  n35 = LinearInterpolate(
   A=LinearInterpolate(
    A=VectorParameter() {group=Color, parameterName=Glass Inner Color, defaultValue=<rgba 0.05 0.05 0.05 1>}.RGB
    B=VectorParameter() {group=Color, parameterName=Glass Outer Color, defaultValue=<rgba 0 0 0 1>}.RGB
    Alpha=Fresnel(ExponentIn=@n34)
   )
   B=VectorParameter() {group=Color, parameterName=Darkening Color, sortPriority=60, defaultValue=<rgba 0 0 0 1>}.RGB
   Alpha=@n27
  )
  n36 = StaticSwitchParameter(True=Multiply(A=@n12, B=@n35), False=@n35) {group=Cracks, parameterName=Is Cracked, sortPriority=0}
  n37 = Abs(None=VertexNormalWS())
  n38 = ComponentMask(None=@n37) {g=true}
  n39 = ComponentMask(None=@n37) {r=true}
  n40 = ScalarParameter() {defaultValue=0.1, group=Dirt, parameterName=Smoothness}
  n41 = ConstantBiasScale(None=Divide(A=Subtract(A=@n38, B=@n39), B=@n40))
  n42 = Divide(A=Subtract(A=ComponentMask(None=@n37) {b=true}, B=Max(A=@n39, B=@n38)), B=@n40)
  n43 = LinearInterpolate(
   A=LinearInterpolate(A=Constant3Vector() {constant=<rgba 1 0 0 0>}, B=Constant3Vector() {constant=<rgba 0 1 0 0>}, Alpha=Saturate(None=@n41))
   B=Constant3Vector() {constant=<rgba 0 0 1 0>}
   Alpha=Saturate(None=ConstantBiasScale(None=@n42))
  )
  n44 = ComponentMask(None=@n43) {r=true}
  n45 = Multiply(A=WorldPosition().XYZ, B=ScalarParameter() {defaultValue=0.1, group=Dirt, parameterName=DirtTextureTile}) {constB=0.1}
  n46 = TextureSampleParameter2D(UVs=ComponentMask(None=@n45) {b=true, g=true}) {
   group=Dirt
   parameterName=Dirt Texture
   samplerType=SAMPLERTYPE_Grayscale
   texture=/Game/Athena/Apollo/Environments/Landscape/Materials/Biome/CommonMaterials/T_ApolloGrass_ColorMask
  }
  n47 = ComponentMask(None=@n43) {g=true}
  n48 = TextureSampleParameter2D(UVs=ComponentMask(None=@n45) {b=true, r=true}) {
   group=Dirt
   parameterName=Dirt Texture
   samplerType=SAMPLERTYPE_Grayscale
   texture=&T_ApolloGrass_ColorMask
  }
  n49 = TextureSampleParameter2D(UVs=ComponentMask(None=@n45) {g=true, r=true}) {
   group=Dirt
   parameterName=Dirt Texture
   samplerType=SAMPLERTYPE_Grayscale
   texture=&T_ApolloGrass_ColorMask
  }
  n50 = Add(A=Add(A=Multiply(A=@n44, B=@n46.R), B=Multiply(A=@n47, B=@n48.R)), B=Multiply(A=ComponentMask(None=@n43) {b=true}, B=@n49.R))
  StaticSwitchParameter(
   True=LinearInterpolate(A=@n36, B=VectorParameter() {group=Dirt, parameterName=DirtColor, defaultValue=<rgba 0.2333 0.1538 0.0515 1>}.RGB, Alpha=@n50)
   False=@n36
  ) {group=Dirt, parameterName=Has Dirt, sortPriority=0}
  n51 = ScalarParameter() {defaultValue=0.2, group=Opacity, parameterName=Fresnel Inner Transparency}
  n52 = ScalarParameter() {defaultValue=0.75, group=Opacity, parameterName=Fresnel Outer Transparency}
  n53 = LinearInterpolate(A=@n51, B=@n52, Alpha=Fresnel(ExponentIn=ScalarParameter() {defaultValue=2, group=Opacity, parameterName=Fresnel Exponent}))
  n54 = LinearInterpolate(A=@n51, B=@n52, Alpha=Fresnel(ExponentIn=ScalarParameter() {defaultValue=2, group=Opacity, parameterName=Fresnel Exponent}, Normal=@n31))
  n55 = StaticSwitchParameter(
   True=StaticSwitchParameter(True=LinearInterpolate(A=@n53, B=@n54, Alpha=@n33), False=@n54) {desc="White = Cracks. Black = No Cracks", group=Cracks, parameterName=Vertex Masking, sortPriority=50}
   False=@n53
  ) {group=Cracks, parameterName=Is Cracked, sortPriority=0}
  n56 = ScalarParameter() {defaultValue=1, group=Dirt, parameterName=DirtTextureContrast}
  StaticSwitchParameter(
   True=LinearInterpolate(
    A=@n55
    Alpha=MaterialFunctionCall!(Color=@n50, Contrast=@n56) {materialFunction=/Engine/Functions/Engine_MaterialFunctions01/ImageAdjustment/Contrast_Preserve_Color}.Result
   )
   False=@n55
  ) {group=Dirt, parameterName=Has Dirt, sortPriority=0}
 ]
}
