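# Additive unlit glow card material for the Spooky alphabet meshes, drawing a tinted soft glow on a scalable camera-facing sprite with adjustable camera offset.
# Core and halo radial masks are summed into opacity, with optional depth fade that mobile skips.
# An optional flicker thresholds three time-driven waves into on/off blinking.
# Pivot-position hashes desynchronize the flicker and drive optional per-instance opacity variation.
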Material /Game/Creative/Environments/Meshes/Spooky/Alphabet/Materials/M_CP_Spooky_Glowcard_01
graph {
 properties {
  bUsedWithNiagaraMeshParticles=true
  blendMode=BLEND_Additive
  pixelDepthOffsetMode=PDOM_Legacy
  shadingModel=MSM_Unlit
  usedShadingModels=Unlit
 }
 nodes [
  n1 = ScalarParameter() {defaultValue=10, group=Scale, parameterName=ScaleX}
  n2 = ScalarParameter() {defaultValue=10, group=Scale, parameterName=ScaleY}
  Add(
   A=MaterialFunctionCall!(Center Location=ObjectPositionWS(), 0-1 UVs, XY scale=AppendVector(A=@n1, B=@n2), OverridePivot, "Normals (Optional)") {materialFunction=&Sprite}.World position offset
   B=MaterialFunctionCall(Offset Amount=ScalarParameter() {group=WPO, parameterName=CameraOffset}) {materialFunction=/Engine/Functions/Engine_MaterialFunctions02/WorldPositionOffset/CameraOffset}.Camera Offset
  )
  n3 = VectorParameter() {group=Color, parameterName=Color, defaultValue=<rgba 1 1 1 1>}
  Reroute(None=@n3.RGB)
  n4 = TextureCoordinate()
  n5 = Constant() {r=0.5}
  n6 = ScalarParameter() {defaultValue=0.5, group=CORE, parameterName=Core_Radius}
  n7 = ScalarParameter() {group=CORE, parameterName=Core_Hardness}
  n8 = Multiply(
   A=Power(
    Base=SphereMask(A=@n4, B=@n5, Radius=@n6, Hardness=@n7) {hardnessPercent=0}
    Exp=ScalarParameter() {defaultValue=5, group=CORE, parameterName=Core_GrandientFalloff}
   )
   B=ScalarParameter() {defaultValue=1, group=CORE, parameterName=Core_Opacity}
  )
  n9 = TextureCoordinate()
  n10 = Constant() {r=0.5}
  n11 = ScalarParameter() {defaultValue=0.5, group=HALO, parameterName=Halo_Radius}
  n12 = ScalarParameter() {group=HALO, parameterName=Halo_Hardness}
  n13 = Multiply(
   A=Power(
    Base=SphereMask(A=@n9, B=@n10, Radius=@n11, Hardness=@n12) {hardnessPercent=0}
    Exp=ScalarParameter() {defaultValue=5, group=HALO, parameterName=Halo_GrandientFalloff}
   )
   B=ScalarParameter() {defaultValue=1, group=HALO, parameterName=Halo_Opacity}
  )
  n14 = Multiply(
   A=Multiply(A=@n3.A, B=ScalarParameter() {defaultValue=1, group=Opacity Settings, parameterName=OpacityGlobal})
   B=Saturate(None=Add(A=@n8, B=@n13))
  )
  n15 = ScalarParameter() {defaultValue=100, group=Opacity Settings, parameterName=DepthFadeDistance}
  n16 = LinearInterpolate(
   A=@n14
   B=ShadingPathSwitch(Default=DepthFade(Opacity=@n14, FadeDistance=@n15), Mobile=@n14) {desc=Migrated from FeatureLevelSwitch}
   Alpha=ScalarParameter() {group=Opacity Settings, parameterName=UseDepthFade?, sliderMax=1}
  )
  n17 = MaterialFunctionCall() {materialFunction=&ObjectPivotPoint}
  n18 = Add(
   A=Divide(A=ComponentMask(None=@n17.Object Pivot Location) {r=true}) {constB=0.577}
   B=Divide(A=ComponentMask(None=@n17.Object Pivot Location) {g=true}) {constB=0.157}
  )
  n19 = Add(A=@n18, B=Divide(A=ComponentMask(None=@n17.Object Pivot Location) {b=true}) {constB=0.296})
  n20 = Multiply(A=Time(), B=Frac(None=@n19))
  n21 = ScalarParameter() {defaultValue=0.05, group=Flicker, parameterName=wave1speed}
  n22 = Sine(None=Subtract(A=Multiply(A=@n20, B=@n21), B=@n18))
  n23 = Frac(None=Time())
  n24 = MaterialFunctionCall() {materialFunction=&ObjectPivotPoint}
  n25 = Divide(A=ComponentMask(None=@n24.Object Pivot Location) {r=true}) {constB=0.437}
  n26 = Divide(A=ComponentMask(None=@n24.Object Pivot Location) {g=true}) {constB=0.578}
  n27 = ComponentMask(None=@n24.Object Pivot Location) {b=true}
  n28 = Frac(None=Add(A=Add(A=@n25, B=@n26), B=Divide(A=@n27) {constB=0.186}))
  n29 = Cosine(None=Multiply(A=Multiply(A=@n23, B=@n28), B=ScalarParameter() {defaultValue=0.1, group=Flicker, parameterName=wave2speed}))
  n30 = Frac(None=Subtract(A=Frac(None=@n22), B=Frac(None=@n29)))
  n31 = MaterialFunctionCall() {materialFunction=&ObjectPivotPoint}
  n32 = Add(
   A=Divide(A=ComponentMask(None=@n31.Object Pivot Location) {r=true}) {constB=0.118}
   B=Divide(A=ComponentMask(None=@n31.Object Pivot Location) {g=true}) {constB=0.988}
  )
  n33 = Divide(A=ComponentMask(None=@n31.Object Pivot Location) {b=true}) {constB=0.387}
  n34 = Multiply(A=Frac(None=Time()), B=Frac(None=Add(A=@n32, B=@n33)))
  n35 = ScalarParameter() {defaultValue=0.25, group=Flicker, parameterName=wave3speed}
  n36 = Frac(None=Cosine(None=Multiply(A=@n34, B=@n35)))
  n37 = Saturate(None=Frac(None=Subtract(A=@n30, B=@n36)))
  n38 = Constant() {r=0.5}
  n39 = Constant() {r=1}
  n40 = Constant()
  n41 = StaticSwitchParameter(True=Multiply(A=@n16, B=If(A=@n37, B=@n38, "A > B"=@n39, "A == B"=@n39, "A < B"=@n40)), False=@n16) {group=Flicker, parameterName=Flicker}
  n42 = Add(A=@n19) {constB=1.2346}
  n43 = Saturate(None=Frac(None=Multiply(A=@n42) {constB=3.1259}))
  n44 = Constant()
  n45 = Constant() {r=1}
  n46 = ScalarParameter() {defaultValue=0.8, group=Variability, parameterName=Variability Floor}
  n47 = Constant() {r=1}
  StaticSwitchParameter(
   True=Multiply(
    A=@n41
    B=MaterialFunctionCall(Input=@n43, Input Low=@n44, Input High=@n45, Target Low=@n46, Target High=@n47) {materialFunction=&RemapValueRange}.Result
   )
   False=@n41
  ) {group=Variability, parameterName=Variability}
 ]
}
