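# Translucent unlit material for spooky bulb filaments, with a tintable light color, intensity and an optional per-object flicker.
# Hashes of the object pivot position vary three time-driven sine and cosine waves, so bulbs at different positions flicker differently.
# The waves' fractional parts are subtracted and thresholded into a hard on/off mask that multiplies the bulb color.
# The Flicker static switch selects between this flickering color and the steady one.
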Material /Game/Creative/Environments/Meshes/Spooky/Alphabet/Materials/M_Spooky_Bulb_Filament_01
graph {
 properties {blendMode=BLEND_Translucent, pixelDepthOffsetMode=PDOM_Legacy, shadingModel=MSM_Unlit, usedShadingModels=Unlit}
 nodes [
  Add()
  Add()
  Add()
  n1 = Time()
  Frac(None=@n1)
  Multiply()
  ObjectPositionWS()
  ObjectPositionWS()
  ObjectPositionWS()
  n2 = Frac(None=Time())
  n3 = MaterialFunctionCall() {materialFunction=&ObjectPivotPoint}
  n4 = Divide(A=ComponentMask(None=@n3.Object Pivot Location) {r=true}) {constB=0.0437}
  n5 = Divide(A=ComponentMask(None=@n3.Object Pivot Location) {g=true}) {constB=0.0578}
  n6 = ComponentMask(None=@n3.Object Pivot Location) {b=true}
  n7 = Frac(None=Add(A=Add(A=@n4, B=@n5), B=Divide(A=@n6) {constB=0.0186}))
  n8 = Cosine(None=Multiply(A=Multiply(A=@n2, B=@n7), B=ScalarParameter() {defaultValue=0.1, group=Flicker, parameterName=wave2speed}))
  OneMinus(None=@n8)
  n9 = Frac(None=Time())
  n10 = MaterialFunctionCall() {materialFunction=&ObjectPivotPoint}
  n11 = Divide(A=ComponentMask(None=@n10.Object Pivot Location) {r=true}) {constB=0.0118}
  n12 = Divide(A=ComponentMask(None=@n10.Object Pivot Location) {g=true}) {constB=0.0988}
  n13 = ComponentMask(None=@n10.Object Pivot Location) {b=true}
  n14 = Frac(None=Add(A=Add(A=@n11, B=@n12), B=Divide(A=@n13) {constB=0.0387}))
  n15 = Cosine(None=Multiply(A=Multiply(A=@n9, B=@n14), B=ScalarParameter() {defaultValue=0.25, group=Flicker, parameterName=wave3speed}))
  OneMinus(None=@n15)
  ScalarParameter() {defaultValue=1, group=Bulb Stuff, parameterName=Opacity, sortPriority=3}
  n16 = Multiply(
   A=VectorParameter() {group=Bulb Stuff, parameterName=Bulb Light Color, sortPriority=1, defaultValue=<rgba 1 1 0 1>}.RGB
   B=ScalarParameter() {defaultValue=1, group=Bulb Stuff, parameterName=Bulb Light Intensity, sortPriority=2}
  )
  n17 = MaterialFunctionCall() {materialFunction=&ObjectPivotPoint}
  n18 = Add(
   A=Divide(A=ComponentMask(None=@n17.Object Pivot Location) {r=true}) {constB=0.0577}
   B=Divide(A=ComponentMask(None=@n17.Object Pivot Location) {g=true}) {constB=0.0157}
  )
  n19 = Divide(A=ComponentMask(None=@n17.Object Pivot Location) {b=true}) {constB=0.0296}
  n20 = Multiply(A=@n1, B=Frac(None=Add(A=@n18, B=@n19)))
  n21 = ScalarParameter() {defaultValue=0.05, group=Flicker, parameterName=wave1speed}
  n22 = Sine(None=Subtract(A=Multiply(A=@n20, B=@n21), B=@n18))
  n23 = Frac(None=Subtract(A=Frac(None=@n22), B=Frac(None=@n8)))
  n24 = Frac(None=@n15)
  n25 = Saturate(None=Frac(None=Subtract(A=@n23, B=@n24)))
  n26 = Constant() {r=0.5}
  n27 = Constant() {r=1}
  n28 = Constant()
  StaticSwitchParameter(True=Multiply(A=@n16, B=If(A=@n25, B=@n26, "A > B"=@n27, "A == B"=@n27, "A < B"=@n28)), False=@n16) {group=Flicker, parameterName=Flicker}
  WorldPosition()
  WorldPosition()
  WorldPosition()
 ]
}
